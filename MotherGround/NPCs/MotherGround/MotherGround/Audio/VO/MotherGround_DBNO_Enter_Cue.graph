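# Sound cue for the MotherGround NPC's voice line when it enters the down-but-not-out (DBNO) state.
# A Random node picks one of four equally weighted AlienLeader DBNO-enter wave files and is excluded from branch culling.
# It plays through the NPC VO sound class with close-range MotherGround attenuation and a chatter concurrency group that governs overlapping lines.
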SoundCue /MotherGround/NPCs/MotherGround/MotherGround/Audio/VO/MotherGround_DBNO_Enter_Cue
root=Random_1
nodes [
 {
  id=Random_1
  children [WavePlayer_0,WavePlayer_4,WavePlayer_5,WavePlayer_6]
  properties {bShouldExcludeFromBranchCulling=true, weights [1,1,1,1]}
 }
 {id=WavePlayer_0, properties {soundWaveAssetPtr=/MotherGround/NPCs/MotherGround/MotherGround/Audio/VO/Waves/AlienLeader_2021_DBNO_Enter_010}}
 {id=WavePlayer_4, properties {soundWaveAssetPtr=/MotherGround/NPCs/MotherGround/MotherGround/Audio/VO/Waves/AlienLeader_2021_DBNO_Enter_011}}
 {id=WavePlayer_5, properties {soundWaveAssetPtr=/MotherGround/NPCs/MotherGround/MotherGround/Audio/VO/Waves/AlienLeader_2021_DBNO_Enter_012}}
 {id=WavePlayer_6, properties {soundWaveAssetPtr=/MotherGround/NPCs/MotherGround/MotherGround/Audio/VO/Waves/AlienLeader_2021_DBNO_Enter_013}}
]
properties {
 attenuationSettings=&/MotherGround_Close_Att
 maxDistance=4000
 soundClassObject=&Fort_VO_NPC
 volumeMultiplier=1
 concurrencySet [&/MotherGround_Chatter_Conc]
}
