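# Sound cue for the MotherGround NPC's voice line, named for losing its target while at full alert.
# A Random node picks among five equally weighted AlienLeader wave variations and is excluded from branch culling.
# It plays in the NPC VO sound class with MotherGround close-range attenuation and a chatter concurrency setting.
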SoundCue /MotherGround/NPCs/MotherGround/MotherGround/Audio/VO/MotherGround_FullAlertLostTarget_Cue
root=Random_0
nodes [
 {
  id=Random_0
  children [WavePlayer_2,WavePlayer_3,WavePlayer_5,WavePlayer_6,WavePlayer_7]
  properties {bShouldExcludeFromBranchCulling=true, weights [1,1,1,1,1]}
 }
 {
  id=WavePlayer_2
  properties {soundWaveAssetPtr=/MotherGround/NPCs/MotherGround/MotherGround/Audio/VO/Waves/AlienLeader_2021_FullAlertLostTarget_011}
 }
 {
  id=WavePlayer_3
  properties {soundWaveAssetPtr=/MotherGround/NPCs/MotherGround/MotherGround/Audio/VO/Waves/AlienLeader_2021_FullAlertLostTarget_020}
 }
 {
  id=WavePlayer_5
  properties {soundWaveAssetPtr=/MotherGround/NPCs/MotherGround/MotherGround/Audio/VO/Waves/AlienLeader_2021_FullAlertLostTarget_031}
 }
 {
  id=WavePlayer_6
  properties {soundWaveAssetPtr=/MotherGround/NPCs/MotherGround/MotherGround/Audio/VO/Waves/AlienLeader_2021_FullAlertLostTarget_041}
 }
 {
  id=WavePlayer_7
  properties {soundWaveAssetPtr=/MotherGround/NPCs/MotherGround/MotherGround/Audio/VO/Waves/AlienLeader_2021_FullAlertLostTarget_051}
 }
]
properties {
 attenuationSettings=&/MotherGround_Close_Att
 maxDistance=4000
 soundClassObject=&Fort_VO_NPC
 volumeMultiplier=1
 concurrencySet [&/MotherGround_Chatter_Conc]
}
