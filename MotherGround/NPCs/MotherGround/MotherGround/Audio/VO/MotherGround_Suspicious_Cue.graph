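# Sound cue for the MotherGround NPC's suspicious voice barks, using AlienLeader suspicious voice waves.
# A Random node picks one of four wave players with equal weight and is excluded from branch culling.
# The cue plays slightly boosted on the NPC voice sound class with close-range attenuation and a chatter concurrency group.
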SoundCue /MotherGround/NPCs/MotherGround/MotherGround/Audio/VO/MotherGround_Suspicious_Cue
root=Random_0
nodes [
 {
  id=Random_0
  children [WavePlayer_0,WavePlayer_3,WavePlayer_4,WavePlayer_5]
  properties {bShouldExcludeFromBranchCulling=true, weights [1,1,1,1]}
 }
 {id=WavePlayer_0, properties {soundWaveAssetPtr=/MotherGround/NPCs/MotherGround/MotherGround/Audio/VO/Waves/AlienLeader_2021_Suspicious_010}}
 {id=WavePlayer_3, properties {soundWaveAssetPtr=/MotherGround/NPCs/MotherGround/MotherGround/Audio/VO/Waves/AlienLeader_2021_Suspicious_041}}
 {id=WavePlayer_4, properties {soundWaveAssetPtr=/MotherGround/NPCs/MotherGround/MotherGround/Audio/VO/Waves/AlienLeader_2021_Suspicious_050}}
 {id=WavePlayer_5, properties {soundWaveAssetPtr=/MotherGround/NPCs/MotherGround/MotherGround/Audio/VO/Waves/AlienLeader_2021_Suspicious_051}}
]
properties {
 attenuationSettings=&/MotherGround_Close_Att
 maxDistance=4000
 soundClassObject=&Fort_VO_NPC
 volumeMultiplier=1.2
 concurrencySet [&/MotherGround_Chatter_Conc]
}
