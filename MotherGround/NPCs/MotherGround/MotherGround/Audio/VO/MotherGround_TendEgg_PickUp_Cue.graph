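# Sound cue for MotherGround NPC voice, played by name for a tend-egg pick-up line.
# A single wave player feeds one AlienLeader tend-egg pick-up recording straight to the output, with no randomization.
# It uses the NPC VO sound class, the MotherGround idle attenuation and the shared chatter concurrency group.
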SoundCue /MotherGround/NPCs/MotherGround/MotherGround/Audio/VO/MotherGround_TendEgg_PickUp_Cue
root=WavePlayer_0
nodes [{
  id=WavePlayer_0
  properties {soundWaveAssetPtr=/MotherGround/NPCs/MotherGround/MotherGround/Audio/VO/Waves/AlienLeader_2021_TendEgg_PickUp_030}
 }]
properties {
 attenuationSettings=&/MotherGround_Idle_Att
 maxDistance=4000
 soundClassObject=&Fort_VO_NPC
 volumeMultiplier=1
 concurrencySet [&/MotherGround_Chatter_Conc]
}
