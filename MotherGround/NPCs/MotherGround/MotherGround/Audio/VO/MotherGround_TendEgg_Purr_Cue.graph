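# Sound cue for the MotherGround NPC's purr vocalization while tending an egg.
# A random node picks one of three equally weighted AlienLeader TendEgg wave recordings.
# Playback uses the MotherGround idle attenuation, the NPC voice sound class and a shared chatter concurrency set.
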SoundCue /MotherGround/NPCs/MotherGround/MotherGround/Audio/VO/MotherGround_TendEgg_Purr_Cue
root=Random_1
nodes [
 {id=Random_1, children [WavePlayer_0,WavePlayer_6,WavePlayer_7], properties {bShouldExcludeFromBranchCulling=true, weights [1,1,1]}}
 {id=WavePlayer_0, properties {soundWaveAssetPtr=/MotherGround/NPCs/MotherGround/MotherGround/Audio/VO/Waves/AlienLeader_2021_TendEgg_010}}
 {id=WavePlayer_6, properties {soundWaveAssetPtr=/MotherGround/NPCs/MotherGround/MotherGround/Audio/VO/Waves/AlienLeader_2021_TendEgg_011}}
 {id=WavePlayer_7, properties {soundWaveAssetPtr=/MotherGround/NPCs/MotherGround/MotherGround/Audio/VO/Waves/AlienLeader_2021_TendEgg_020}}
]
properties {
 attenuationSettings=&/MotherGround_Idle_Att
 maxDistance=4000
 soundClassObject=&Fort_VO_NPC
 volumeMultiplier=1
 concurrencySet [&/MotherGround_Chatter_Conc]
}
